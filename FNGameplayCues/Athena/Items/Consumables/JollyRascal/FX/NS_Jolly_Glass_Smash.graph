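# Niagara effect for the Jolly Rascal consumable, a brief burst of breaking glass tinted by the User.FXColor parameter.
# GlassShards bursts spinning, randomly sized sprites from a small sphere, thrown mostly along +X under heavy gravity, flashing bright then shrinking.
# GlassSpecks adds smaller GPU-simulated sprites under lighter gravity, twinkling through an oscillating size curve with boosted brightness and an alpha fade.
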NiagaraSystem /FNGameplayCues/Athena/Items/Consumables/JollyRascal/FX/NS_Jolly_Glass_Smash
dependencies {usedDataInterfaces [&NiagaraDataInterfaceCurve], usedDynamicInputs [], usedModules [&SystemState]}
emitters {
 GlassShards {
  emitterUpdateScript [
   &EmitterState(Life Cycle Mode=::NewEnumerator1, Loop Behavior=::NewEnumerator1, Loop Duration=0.12, UseLoopDelay=true)
   &SpawnBurst_Instantaneous(Spawn Count=24, Use Spawn Probability=false, UseLoopCountLimit=false)
  ]
  particleSpawnScript [
   /SetVariables(Particles.SpriteRotationRate=&RandomRangeFloat(Minimum=-4, Maximum=4, Evaluation Type=::NewEnumerator1)) {isSetParameters=true}
   &InitializeParticle(
    Color=$User.FXColor
    Lifetime=&RandomRangeFloat(Minimum=0.4, Maximum=1.25, Evaluation Type=::NewEnumerator1)
    Sprite Rotation Angle=&RandomRangeFloat(Minimum=0, Maximum=1, Evaluation Type=::NewEnumerator1)
    Sprite Rotation Mode=::NewEnumerator1
    Sprite Size=&RandomRangeVector2D(Minimum=<4 4>, Maximum=<15 15>, Evaluation Type=::NewEnumerator1)
    Sprite Size Mode=::NewEnumerator4
    UsePositionOffset=false
    Write Material Random=true
   )
   &ShapeLocation(MaskSpawnGroup=false, Override Seed=false, Sphere Radius=15, UseOwnerScale=true)
   &AddVelocity(Velocity=&Multiply_VectorByFloat(
     Vector=$Particles.ShapeLocation.ShapeVector
     Float=&RandomRangeFloat(Minimum=20, Maximum=30, Evaluation Type=::NewEnumerator1)
    ))
   &AddVelocity(Velocity=&RandomRangeVector(Minimum=<200 -40 -40>, Maximum=<400 40 40>, Evaluation Type=::NewEnumerator1)) {module=AddVelocity001}
  ]
  particleUpdateScript [
   &ParticleState()
   &SpriteRotationRate(Rotation Rate=&AngleConversion(
     Angle:NiagaraFloat=$Particles.SpriteRotationRate
     Angle Input:ENiagara_AngleInput=::NewEnumerator1
     Angle Output:ENiagara_AngleInput=::NewEnumerator0
    ))
   &ScaleSpriteSize(
    Scale Factor=&Vector2DFromFloat(Value=&FloatFromCurve(FloatCurve=DataInterface() {propertyValues {Curve {keys [<at 0.25 1 RCIM_Cubic>,<at 1 0>]}}}))
    Scale Sprite Size Mode=::NewEnumerator2
   )
   &AccelerationForce(Acceleration=<0 0 -1600>)
   &SolveForcesAndVelocity(Clamp Velocity=false, Limit Acceleration=false)
   &ScaleColor(
    Scale RGB=&VectorFromFloat(Value=&FloatFromCurve(FloatCurve=DataInterface() {propertyValues {Curve {keys [
          {interpMode=RCIM_Cubic, leaveTangent=-63.2591, tangentMode=RCTM_Break, time=0, value=15}
          {interpMode=RCIM_Cubic, tangentMode=RCTM_Break, time=0.25, value=1}
         ]}}}))
    ScaleA=true
    ScaleRGB=true
   )
   &CameraOffset(Camera Offset Amount=15)
  ]
  rendererClasses [&NiagaraSpriteRendererProperties]
  renderers [{rendererClass=&NiagaraSpriteRendererProperties}]
 }
 GlassSpecks {
  simTarget=GPUComputeSim
  emitterUpdateScript [
   &EmitterState(Life Cycle Mode=::NewEnumerator1, Loop Behavior=::NewEnumerator1, Loop Duration=0.1, UseLoopDelay=true)
   &SpawnBurst_Instantaneous(Spawn Count=32, Use Spawn Probability=false, UseLoopCountLimit=false)
  ]
  particleSpawnScript [
   &InitializeParticle(
    Color=$User.FXColor
    Lifetime Max=1
    Lifetime Min=0.25
    Lifetime Mode=::NewEnumerator1
    Sprite Size=&Vector2DFromFloat(Value=&RandomRangeFloat(Minimum=4, Maximum=8, Evaluation Type=::NewEnumerator1))
    Sprite Size Mode=::NewEnumerator4
    UsePositionOffset=false
    Write Material Random=true
   )
   &AddVelocity(Velocity=&RandomRangeVector(Minimum=<100 -20 -20>, Maximum=<300 20 20>, Evaluation Type=::NewEnumerator1))
   &ShapeLocation(MaskSpawnGroup=false, Override Seed=false, Sphere Radius=15, UseOwnerScale=true)
   &AddVelocity(Velocity=&Multiply_VectorByFloat(
     Vector=$Particles.ShapeLocation.ShapeVector
     Float=&RandomRangeFloat(Minimum=20, Maximum=40, Evaluation Type=::NewEnumerator1)
    )) {module=AddVelocity001}
  ]
  particleUpdateScript [
   &ParticleState()
   &ScaleSpriteSize(
    Scale Sprite Size Mode=::NewEnumerator0
    Uniform Scale Factor=&FloatFromCurve(FloatCurve=DataInterface() {propertyValues {Curve {keys [
         {arriveTangent=-6.1151, interpMode=RCIM_Cubic, time=0, value=1}
         {arriveTangent=-0.2183, interpMode=RCIM_Cubic, leaveTangent=-0.2183, time=0.1354, value=0.2886}
         {arriveTangent=-0.5986, interpMode=RCIM_Cubic, leaveTangent=-0.5986, time=0.2671, value=0.9417}
         {arriveTangent=0.057, interpMode=RCIM_Cubic, leaveTangent=0.057, time=0.412, value=0.1231}
         {arriveTangent=0.021, interpMode=RCIM_Cubic, leaveTangent=0.021, time=0.5725, value=0.9591}
         {arriveTangent=-0.1312, interpMode=RCIM_Cubic, leaveTangent=-0.1312, time=0.7149, value=0.1294}
         {arriveTangent=-0.4539, interpMode=RCIM_Cubic, leaveTangent=-0.4539, time=0.8709, value=0.9199}
         <at 1 0 RCIM_Cubic>
        ]}}})
   )
   &ScaleColor(
    Scale Alpha=&FloatFromCurve(FloatCurve=DataInterface() {propertyValues {Curve {keys [
         {arriveTangent=9.5296, interpMode=RCIM_Cubic, time=0, value=0}
         {arriveTangent=4.1283, interpMode=RCIM_Cubic, leaveTangent=4.1283, time=0.1068, value=0.9998}
         {arriveTangent=-0.0193, interpMode=RCIM_Cubic, leaveTangent=-0.0193, time=0.2549, value=1.0522}
         {arriveTangent=-0.6558, interpMode=RCIM_Cubic, leaveTangent=-0.6558, time=0.4085, value=0.994}
         {arriveTangent=-1.2828, interpMode=RCIM_Cubic, leaveTangent=-1.2828, time=0.5881, value=0.8337}
         {arriveTangent=-4.0521, interpMode=RCIM_Cubic, leaveTangent=-4.0521, time=0.6717, value=0.6564}
         {arriveTangent=-1.9991, interpMode=RCIM_Cubic, leaveTangent=-1.9991, time=0.7718, value=0.0893}
         <at 1 0>
        ]}}})
    Scale RGB=<7 7 7>
    ScaleA=true
    ScaleRGB=true
   )
   &AccelerationForce(Acceleration=<0 0 -800>)
   &SolveForcesAndVelocity(Clamp Velocity=false, Limit Acceleration=false)
  ]
  rendererClasses [&NiagaraSpriteRendererProperties]
  renderers [{rendererClass=&NiagaraSpriteRendererProperties}]
 }
}
summary {
 systemName=NS_Jolly_Glass_Smash
 emitters [
  {emitterName=GlassSpecks, simTarget=GPUComputeSim, rendererClasses [&NiagaraSpriteRendererProperties]}
  {emitterName=GlassShards, rendererClasses [&NiagaraSpriteRendererProperties]}
 ]
 userVariables [User.FXColor {type=LinearColor, description="", defaultValue {struct=LinearColor, value=<rgba 0.2039 0.3176 0.502 1>}}]
}
